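# Post-process AnimBlueprint for the TaperSlink large male head cosmetic that corrects head, neck and upper-spine pose.
# Each update flags helmet visibility from the material variant and remaps head-to-spine_05 delta rotation into corrective rotators.
# Correction weights vary with emote state, while front-end and parachute states drive additional values.
# AnimGraph caches a scale and spine adjusted pose, then bool-blends it against additional additive neck and head offsets.
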AnimBlueprint /BRCosmetics/Characters/Player/Male/Large/Heads/M_LRG_TaperSlink_Head/Meshes/M_LRG_TaperSlink_Head_Post_Process
target_skeleton=/Game/Characters/Player/Common/Fortnite_Base_Head/Fortnite_Base_Head_Export_Skeleton
graphs [
 AnimGraph {nodes [
   n1 = Get Neck Rotate_0_pos() -> (Neck Rotate_0_pos:Rotator)
   n2 = "Transform (Modify) Bone - Bone: neck_01"(
    ComponentPose=Local To Component(LocalPose=Use cached pose 'InputPose Cache'())
    Alpha=1
    Translation="-0.95,0,0"
    Rotation=@n1
    Scale="1,1,1"
   ) {
    rotationMode=BMM_Additive
    rotationSpace=BCS_BoneSpace
    translationMode=BMM_Additive
    translationSpace=BCS_BoneSpace
    alphaScaleBiasClamp {bMapRange=true, inRange {max=0, min=1}}
    boneToModify {boneName=neck_01}
    translation {x=-0.95}
   }
   n3 = "Transform (Modify) Bone - Bone: neck_02"(ComponentPose=@n2, Alpha=1, Translation="-0.95,0,0", Rotation=@n1, Scale="1,1,1") {
    rotationMode=BMM_Additive
    rotationSpace=BCS_BoneSpace
    translationMode=BMM_Additive
    translationSpace=BCS_BoneSpace
    alphaScaleBiasClamp {bMapRange=true, inRange {max=0, min=1}}
    boneToModify {boneName=neck_02}
    translation {x=-0.95}
   }
   n4 = "Transform (Modify) Bone - Bone: neck_01"!(ComponentPose=@n3, bAlphaBoolEnabled:bool=True, Translation="-0.95,0,0", Rotation="0,-4,0", Scale="1,1,1") -> (Pose) {
    alphaInputType=Bool
    rotationMode=BMM_Additive
    rotationSpace=BCS_BoneSpace
    translationMode=BMM_Additive
    translationSpace=BCS_BoneSpace
    alphaBoolBlend {blendInTime=0.25, blendOutTime=0.25}
    alphaScaleBiasClamp {bMapRange=true}
    boneToModify {boneName=neck_01}
    rotation {yaw=-4}
    translation {x=-0.95}
   }
   n5 = "Transform (Modify) Bone - Bone: head"(
    ComponentPose=@n4
    Alpha=1
    Translation="-0.2,-0.5,0"
    Rotation=Get Head Rotate 0 pos() -> (Head Rotate 0 pos:Rotator)
    Scale="0.98,0.98,0.98"
   ) {
    rotationMode=BMM_Additive
    rotationSpace=BCS_BoneSpace
    scaleMode=BMM_Additive
    scaleSpace=BCS_BoneSpace
    translationMode=BMM_Additive
    translationSpace=BCS_BoneSpace
    alphaScaleBiasClamp {bMapRange=true, inRange {max=0, min=1}}
    boneToModify {boneName=head}
    scale=<0.98 0.98 0.98>
    translation=<-0.2 -0.5>
   }
   Output Pose(Result=Blend Poses by bool(BlendPose_0=Component To Local(ComponentPose=@n5), BlendPose_1=Use cached pose 'InputPose Cache'())) {layerGroup=None}
   n6 = "Transform (Modify) Bone - Bone: head"!(
    ComponentPose=Local To Component(LocalPose=Input Pose())
    bAlphaBoolEnabled:bool=True
    Translation="0,0,0"
    Rotation="0,0,0"
    Scale="0.98,0.99,0.99"
   ) -> (Pose) {
    alphaInputType=Bool
    rotationMode=BMM_Additive
    rotationSpace=BCS_BoneSpace
    scaleMode=BMM_Additive
    scaleSpace=BCS_BoneSpace
    alphaBoolBlend {blendInTime=0.3, blendOutTime=0.3}
    boneToModify {boneName=head}
    scale=<0.98 0.99 0.99>
   }
   n7 = "Transform (Modify) Bone - Bone: neck_02"!(ComponentPose=@n6, bAlphaBoolEnabled:bool=True, Translation="0,0,0", Rotation="0,0,0", Scale="0.98,0.99,0.99") -> (Pose) {
    alphaInputType=Bool
    rotationMode=BMM_Additive
    rotationSpace=BCS_BoneSpace
    scaleMode=BMM_Additive
    scaleSpace=BCS_BoneSpace
    alphaBoolBlend {blendInTime=0.3, blendOutTime=0.3}
    boneToModify {boneName=neck_02}
    scale=<0.98 0.99 0.99>
   }
   n8 = "Transform (Modify) Bone - Bone: neck_01"(ComponentPose=@n7, bAlphaBoolEnabled=True, Translation="0,0,0", Rotation="0,0,0", Scale="0.98,0.99,0.99") {
    alphaInputType=Bool
    rotationMode=BMM_Additive
    rotationSpace=BCS_BoneSpace
    scaleMode=BMM_Additive
    scaleSpace=BCS_BoneSpace
    alphaBoolBlend {blendInTime=0.3, blendOutTime=0.3}
    boneToModify {boneName=neck_01}
    scale=<0.98 0.99 0.99>
   }
   n9 = "Transform (Modify) Bone - Bone: spine_05"!(
    ComponentPose=@n8
    bAlphaBoolEnabled:bool=True
    Translation="0,0,0"
    Rotation=Get Spine Rotate 0 Pos() -> (Spine Rotate 0 Pos:Rotator)
    Scale="1,1,1"
   ) -> (Pose) {
    alphaInputType=Bool
    rotationMode=BMM_Additive
    rotationSpace=BCS_BoneSpace
    scaleMode=BMM_Additive
    scaleSpace=BCS_BoneSpace
    alphaBoolBlend {blendInTime=0.3, blendOutTime=0.3}
    boneToModify {boneName=spine_05}
   }
   Save cached pose 'InputPose Cache'(Pose=Component To Local(ComponentPose=@n9)) {cacheName=InputPose Cache}
  ]}
 EventGraph {nodes [
   Event BlueprintUpdateAnimation()
   > n1 = Sequence()
   > n2 = Cast To FortPlayerPawn(Object=GetOwningActor())
   > n3 = GetVariantChannelActiveIndexForItemActor!(
    execute
    Item:Athena Cosmetic Item Definition Object Reference=/BRCosmetics/Athena/Items/Cosmetics/Characters/Character_TaperSlink
    Actor:Actor Object Reference=@n2
    Channel:Gameplay Tag Structure=(TagName="Cosmetics.Variant.Channel.Material")
    VariantList:Array of Gameplay Tag Structures=Make Array!("[0]"=(TagName="Cosmetics.Variant.Property.Mat1"), "[1]":Gameplay Tag Structure=(TagName="Cosmetics.Variant.Property.Mat2")) -> (Array)
    bErrorIfChannelMissing:bool=false
    bErrorIfTagNotInArray:bool=false
   ) -> (then, Index:int=0, bIndexIsValid:bool, bFoundChannel:bool)
   > n4 = Sequence()
   > Set Helmet vis(Helmet vis:bool=AND Boolean(A="Equal (Integer)"(A=@n3), B=@n3.bIndexIsValid)) -> (Output_Get:bool)
   Set Helmet vis(execute=@n4.then_1, Helmet vis=AND Boolean(A="Equal (Integer)"(A=@n3, B=1), B=@n3.bIndexIsValid))
   n5 = Cast To FortPlayerPawn(execute=@n1.then_1, Object=TryGetPawnOwner())
   n6 = Get Mesh(self=@n5)
   n7 = BreakRotator(InRot=BreakTransform(InTransform=GetDeltaTransformFromRefPose(self=@n6, BoneName=head, BaseName=spine_05)).Rotation)
   n8 = MapRangeClamped(Value=@n7.Yaw, InRangeA=-30, InRangeB=40, OutRangeA=-20, OutRangeB=18)
   n9 = Get Z percentage head() -> (Z percentage head:float)
   n10 = GetDeltaTransformFromRefPose(self=@n6, BoneName=head, BaseName=spine_05)
   n11 = float * float!(
    A=BreakRotator(InRot=BreakTransform(InTransform=@n10).Rotation).Pitch
    B=0.1
    C:double=Get Y percentage spine() -> (Y percentage spine:float)
   ) -> (ReturnValue)
   > Set Head Rotate 0 pos(Head Rotate 0 pos:Rotator=MakeRotator(Pitch=@n11, Yaw=float * float(A=@n8, B=-1, C=@n9))) -> (Output_Get:Rotator)
   n12 = MapRangeClamped(Value=@n7.Yaw, InRangeA=-35, InRangeB=15, OutRangeA=-10, OutRangeB=Get Neck range() -> (Neck range:double))
   n13 = Get Z percentage neck() -> (Z percentage neck:float)
   > Set Neck Rotate_0_pos(Neck Rotate_0_pos:Rotator=MakeRotator(Yaw=float * float(A=@n12, B=-1, C=@n13))) -> (Output_Get:Rotator)
   > Set Neck Rotate(Neck Rotate:double=float - float(A=MapRangeClamped(Value=@n7.Yaw, InRangeA=0, InRangeB=-5), B=Get Custom Idle Is Playing Float() -> (Custom Idle Is Playing Float:double))) -> (Output_Get:double)
   > n14 = Sequence()
   > n15 = Branch(Condition=Get bIsPlayingEmote())
   > n16 = Branch(Condition=Montage_IsPlaying(self=GetAnimInstance(self=@n6), Montage=/BRCosmetics/Animation/Game/MainPlayer/Emotes/Engaged_Walk/Engaged_Walk_Traversal_CMM_M))
   > Set Z percentage neck(Z percentage neck:float=0.9) -> (Output_Get:float)
   > Set Z percentage head(Z percentage head:float=0.5) -> (Output_Get:float)
   > Set Neck range(Neck range:double=19) -> (Output_Get:double)
   Set Z percentage neck(execute=@n16.else, Z percentage neck=0.35)
   > Set Z percentage head(Z percentage head=0.3)
   > Set Neck range(Neck range=2)
   Set Z percentage neck(execute=@n15.else, Z percentage neck=0.4)
   > Set Z percentage head(Z percentage head=0.93)
   n17 = Branch(execute=@n14.then_1, Condition=Get bIsInFrontEnd())
   > Set Custom Idle Is Playing Float(Custom Idle Is Playing Float:double=GetCurveValue(self=GetOwnerFrontEndAnimInstance(), CurveName=Frontend_Idle_Bypass)) -> (Output_Get:double)
   Branch(execute=@n17.else, Condition=Get bIsParachuteOpened())
   > Set Spine Rotate 0 Pos(Spine Rotate 0 Pos:Rotator=MakeRotator(Pitch=@n11)) -> (Output_Get:Rotator)
   > Set Custom Idle Is Playing Float(Custom Idle Is Playing Float=0)
  ]}
]
variables [
 Custom Idle Is Playing Float {type=number}
 Head Rotate 0 pos {type=Rotator}
 Head Side Rotate 0 {type=Rotator}
 Head Translation {type=Vector}
 Helmet vis {type=boolean}
 Neck range {type=number}
 Neck Rotate {type=number}
 Neck Rotate_0_pos {type=Rotator}
 Neck Rotate_1_neg {type=Rotator}
 Neck Translation {type=Vector}
 Side head Rot {type=number}
 Spine Rotate 0 Pos {type=Rotator}
 Y percentage neck {type=number, default=0.77}
 Y percentage spine {type=number, default=0.55}
 Z percentage head {type=number, default=0.93}
 Z percentage neck {type=number, default=0.4}
]
